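# Translucent unlit Niagara material for the Astral Radiance drift trail, usable on ribbons, sprites and mesh particles.
# Emissive color combines a panning trail texture with fast-scrolling mask lines, warped by flow noise via Distort and scaled by Lines, then tinted by particle color through an emissive adjustment function.
# Opacity multiplies trail alpha, particle alpha and an optional U-axis mask that fades both ends, toggled by a scalar through an If node.
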Material /VehicleCosmetics/DriftTrail/DriftTrail_AstralRadiance/Materials/M_DriftTrail_OpacityTrail
graph {
 properties {
  bAutomaticallySetUsageInEditor=false
  bUsedWithNiagaraMeshParticles=true
  bUsedWithNiagaraRibbons=true
  bUsedWithNiagaraSprites=true
  blendMode=BLEND_Translucent
  shadingModel=MSM_Unlit
  usedShadingModels=Unlit
 }
 nodes [
  n1 = TextureSample(UVs=Panner(Coordinate=TextureCoordinate() {uTiling=1.6}) {speedX=-2}) {texture=/VehicleCosmetics/Boosts/Boost_AstralRadiance/Textures/trail2}
  n2 = TextureCoordinate() {uTiling=3}
  n3 = TextureSample(UVs=Panner(Coordinate=TextureCoordinate()) {speedX=0.5}) {texture=/DealershipSharedLibrary/FX/Textures/Noise/T_VCVFX_Flow02_N}
  n4 = Multiply(
   A=Multiply(A=ComponentMask(None=@n3.RGB) {g=true, r=true}, B=ScalarParameter() {defaultValue=1, parameterName=Distort})
   B=Constant() {r=0.7}
  )
  n5 = TextureSample(UVs=Panner(Coordinate=Add(A=@n2, B=@n4)) {speedX=-3}) {texture=/VehicleCosmetics/Boosts/Boost_AstralRadiance/Textures/CocoaToon_Mask_T}
  n6 = Subtract(A=Multiply(A=Constant() {r=0.7}, B=@n1.B), B=@n5.R)
  n7 = Multiply(A=@n5.R, B=ScalarParameter() {defaultValue=6, parameterName=Lines})
  MaterialExpressionVertexColor_0 = ParticleColor()
  MaterialFunctionCall(Emissive=Multiply(A=Add(A=@n6, B=@n7), B=@MaterialExpressionVertexColor_0.RGB)) {materialFunction=&MF_FX_EmissiveAdjustments}
  MaterialFunctionCall("Rotation Angle (0-1)"=Constant() {r=0.75}) {materialFunction=&CustomRotator}
  n8 = ComponentMask(None=TextureCoordinate()) {r=true}
  n9 = Power(Base=@n8, Exp=Constant() {r=1.4})
  n10 = Constant() {r=30}
  n11 = OneMinus(None=@n8)
  n12 = Constant() {r=1.6}
  MaterialExpressionConstantClamp_0 = Saturate(None=Multiply(A=Multiply(A=@n9, B=@n10), B=Power(Base=@n11, Exp=@n12)))
  n13 = If(
   A=ScalarParameter() {defaultValue=1, group=Opacity, parameterName=Opacity - Horizontal Mask?}
   "A > B"=@MaterialExpressionConstantClamp_0
   "A == B"=@MaterialExpressionConstantClamp_0
   "A < B"=Constant() {r=1}
  ) {constB=1}
  Reroute(None=Multiply(A=@MaterialExpressionVertexColor_0.A, B=Multiply(A=@n13, B=@n1.A)))
 ]
}
